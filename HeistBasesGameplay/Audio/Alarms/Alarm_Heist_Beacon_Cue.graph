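# Sound cue for a heist base alarm beacon, playing a distant exterior lockdown beacon sound as gameplay event audio.
# A Looping node repeats a fixed Delay into a single Wave Player, leaving a gap before each play.
# Overridden attenuation adds long-range falloff, distance low-pass filtering and occlusion, taking precedence over the assigned alarm attenuation asset.
# Concurrency allows one instance, stopping the farthest first, with distance-based volume scaling.
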SoundCue /HeistBasesGameplay/Audio/Alarms/Alarm_Heist_Beacon_Cue
root=Looping_0
nodes [
 {id=Delay_0, children [WavePlayer_0], properties {delayMax=3, delayMin=3}}
 {id=Looping_0, children [Delay_0]}
 {id=WavePlayer_0, properties {soundWaveAssetPtr=/HeistBasesGameplay/Audio/Alarms/wav/Alarm_Heist_Lockdown_LP_ExteriorBeacon_Dist}}
]
properties {
 attenuationSettings=/Game/Sounds/Asteria/POI/HeistPOIs/Att_Base_Alarms
 bOverrideAttenuation=true
 bOverrideConcurrency=true
 maxDistance=12500
 soundClassObject=&Fort_Events_Gameplay
 volumeMultiplier=0.65
 attenuationOverrides {
  bAttenuateWithLPF=true
  bEnableOcclusion=true
  falloffDistance=2500
  lPFFrequencyAtMax=900
  lPFFrequencyAtMin=15000
  lPFRadiusMax=8000
  lPFRadiusMin=200
  occlusionLowPassFilterFrequency=2000
  attenuationShapeExtents {x=10000}
 }
 concurrencyOverrides {maxCount=1, resolutionRule=StopFarthestThenPreventNew, voiceStealReleaseTime=1, volumeScaleAttackTime=2, volumeScaleMode=Distance}
}
