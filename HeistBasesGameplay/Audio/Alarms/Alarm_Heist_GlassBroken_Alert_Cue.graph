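# Sound cue that plays the alert alarm for a broken-glass event in Heist bases gameplay.
# A single wave player at the root plays the Lockdown one-shot alarm wave, with no randomization, modulation or mixing nodes.
# It routes through the gameplay events sound class and sets a very large max distance.
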SoundCue /HeistBasesGameplay/Audio/Alarms/Alarm_Heist_GlassBroken_Alert_Cue
root=WavePlayer_0
nodes [{id=WavePlayer_0, properties {soundWaveAssetPtr=/HeistBasesGameplay/Audio/Alarms/wav/Alarm_Heist_Lockdown_OneShot}}]
properties {maxDistance=2097152, soundClassObject=&Fort_Events_Gameplay}
